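# Unlit, masked, two-sided material for the AnglePatch arrow emote prop, giving a tinted hologram-style look with Fresnel and scanlines.
# A texture color mask and UV sphere masks shape color and cutout, with top-to-bottom tint and opacity blends.
# Screen-space scanlines bend with view angle and combine with Fresnel, both modulated by near-camera fade, under dithered masking.
# Quality and shading-path switches simplify the Fresnel and bend on cheaper paths.
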Material /BRCosmetics/Animation/Props/Gadgets_AnglePatch/Materials/M_AnglePatch_Arrow_Emote_Prop
graph {
 properties {
  allowTranslucentCustomDepthWrites=true
  bUsedWithMeshDeformer=true
  bUsedWithMorphTargets=true
  bUsedWithSkeletalMesh=true
  blendMode=BLEND_Masked
  ditherOpacityMask=true
  opacityMaskClipValue=0.6
  pixelDepthOffsetMode=PDOM_Legacy
  shadingModel=MSM_Unlit
  twoSided=true
  usedShadingModels=Unlit
 }
 nodes [
  n1 = Add(A=NamedRerouteUsage(), B=NamedRerouteUsage())
  Multiply(A=LinearInterpolate(
    A=VectorParameter() {parameterName=TopTint, defaultValue=<rgba 1 0.632 0 1>}.RGB
    B=VectorParameter() {parameterName=BottomTint, defaultValue=<rgba 1 0.0104 0.0104 1>}.RGB
    Alpha=Saturate(None=@n1)
   )) {constB=2}
  n2 = NamedRerouteUsage()
  n3 = NamedRerouteUsage()
  n4 = NamedRerouteUsage()
  n5 = MaterialFunctionCall(Distance=ScalarParameter() {defaultValue=1500, parameterName=FadeDistance}) {materialFunction=&Near_Camera_Fade}
  n6 = Constant() {r=3}
  n7 = QualitySwitch(Default=LinearInterpolate(Alpha=@n5.Result) {constA=3.2}, Low=@n6)
  n8 = Add(
   A=Fresnel(
    ExponentIn=ShadingPathSwitch(Default=@n7, Forward=@n6)
    BaseReflectFractionIn=ScalarParameter() {defaultValue=0.3, parameterName=BaseRefract}
   ) {exponent=3}
   B=ScalarParameter() {defaultValue=0.4, parameterName=ScanlineAdd}
  )
  n9 = DotProduct(A=ComponentMask(None=VertexNormalWS()) {b=true}, B=ComponentMask(None=CameraVectorWS()) {b=true})
  n10 = ScalarParameter() {defaultValue=0.5, parameterName=LineBend}
  n11 = Constant()
  n12 = QualitySwitch(Default=Multiply(A=Fresnel() {baseReflectFraction=0, exponent=3}, B=Multiply(A=@n9, B=@n10)) {constB=0.2}, Low=@n11, Medium=@n11)
  n13 = ComponentMask(None=ScreenPosition().ViewportUV) {g=true}
  n14 = Sine(None=Multiply(A=Add(A=@n12, B=@n13), B=ScalarParameter() {defaultValue=200, parameterName=ScanlineTile}))
  n15 = LinearInterpolate(A=Power(Base=ConstantBiasScale(None=@n14)), Alpha=@n5.Result) {constA=0.4, constB=0.3}
  Multiply(
   A=LinearInterpolate(
    A=ScalarParameter() {parameterName=TopOpacity}
    B=ScalarParameter() {defaultValue=0.75, parameterName=BottomOpacity}
    Alpha=LinearInterpolate(A=@n2, B=@n3, Alpha=@n4)
   )
   B=Clamp(None=Add(A=@n8, B=@n15) {constB=0.2})
  )
  n16 = TextureSample() {texture=/BRCosmetics/Animation/Props/Gadgets_AnglePatch/Textures/T_Arrow_ColorMask}
  NamedRerouteDeclaration(None=@n16.R) {name=R, nodeColor=<rgba 0.125 0 1 1>}
  NamedRerouteDeclaration(None=@n16.G) {name=G, nodeColor=<rgba 0.1953 1 0 1>}
  n17 = SphereMask(A=TextureCoordinate(), B=AppendVector(A=Constant(), B=Constant() {r=1}), Radius=Constant() {r=0.52}) {hardnessPercent=95}
  NamedRerouteDeclaration(None=Multiply(A=OneMinus(None=@n17), B=NamedRerouteUsage())) {name=Opacity Mask, nodeColor=<rgba 1 0 0.5156 1>}
  n18 = SphereMask(A=TextureCoordinate(), B=AppendVector(A=Constant(), B=Constant() {r=1}), Radius=Constant() {r=0.65}, Hardness=Constant() {r=0.7}) {hardnessPercent=80}
  NamedRerouteDeclaration(None=Multiply(A=NamedRerouteUsage(), B=OneMinus(None=@n18))) {name=Color mask, nodeColor=<rgba 0 0.8359 1 1>}
 ]
}
