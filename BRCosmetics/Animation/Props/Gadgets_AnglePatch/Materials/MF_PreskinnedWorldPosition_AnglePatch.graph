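# Material function in the Chaos library category that outputs preskinned world position, so Chaos Destructible assets keep world-tiled textures with per-body offsets while fracturing and simulating.
# Vertex-interpolated preskinned position passes through Transform3x3Matrix with actor position as origin.
# All three basis inputs are wired to the component's X vector.
# A side branch samples a virtual normal map using the scaled position and converts it world-to-local, without reaching the output.
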MaterialFunction /BRCosmetics/Animation/Props/Gadgets_AnglePatch/Materials/MF_PreskinnedWorldPosition_AnglePatch
graph {
 properties {
  bExposeToLibrary=true
  description="Preskinned World Position - Allows Chaos Destructible assets to apply a world tiled texture, while still maintaining per body offsets during fracturing and simulation."
  libraryCategoriesText [Chaos]
 }
 nodes [
  n1 = MaterialFunctionCall() {materialFunction=/Engine/Functions/Engine_MaterialFunctions02/WorldPositionOffset/ComponentBasisVectors}
  n2 = MaterialFunctionCall(VectorToTransform=VertexInterpolator(VS=PreSkinnedPosition()).PS, BasisX=@n1.X, BasisY=@n1.X, BasisZ=@n1.X, BasisPosition=ActorPositionWS()) {materialFunction=&Transform3x3Matrix}
  FunctionOutput(None=@n2.Transformed Vector)
  n3 = Constant() {r=200}
  Transform(None=TextureSample(UVs=Divide(A=@n2.Transformed Vector, B=@n3)) {samplerType=SAMPLERTYPE_VirtualNormal}.RGB) {transformSourceType=TRANSFORMSOURCE_World, transformType=TRANSFORM_Local}
 ]
}
