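# Translucent, unlit, two-sided hologram material for the skeletal AnglePatch gadget emote prop, drawing glowing contour lines and a grid.
# Contour bands come from pre-skinned height offset by the normal map, antialiased with screen derivatives and widened by depth.
# Packed FX texture channels, a grid texture and Fresnel build masks for tinted emissive and opacity.
# A switchable custom depth comparison masks pixels for sorting, bypassed on mobile.
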Material /BRCosmetics/Animation/Props/Gadgets_AnglePatch/Materials/M_AnglePatch_Hologram_Emote_Prop
graph {
 properties {
  bAutomaticallySetUsageInEditor=false
  bUsedWithMeshDeformer=true
  bUsedWithSkeletalMesh=true
  blendMode=BLEND_Translucent
  ditherOpacityMask=true
  shadingModel=MSM_Unlit
  twoSided=true
  usedShadingModels=Unlit
 }
 nodes [
  n1 = DDX(Value=NamedRerouteUsage())
  n2 = DDY(Value=NamedRerouteUsage())
  n3 = Saturate(None=Divide(A=PixelDepth().R) {constB=5000})
  n4 = Multiply(A=Add(A=Abs(None=@n1), B=Abs(None=@n2)), B=LinearInterpolate(A=Constant() {r=1.5}, B=Constant() {r=5}, Alpha=@n3) {constB=10})
  n5 = Subtract(A=NamedRerouteUsage(), B=@n4)
  n6 = Add(A=NamedRerouteUsage(), B=@n4)
  n7 = LinearInterpolate(A=Constant() {r=0.04}, B=Constant() {r=5}, Alpha=@n3) {constB=20}
  n8 = Saturate(None=Multiply(A=SmoothStep(Min=@n5, Max=@n6, Value=@n7)) {constB=1.2})
  n9 = Max(A=NamedRerouteUsage(), B=@n8)
  n10 = NamedRerouteUsage()
  n11 = NamedRerouteUsage()
  n12 = Constant()
  n13 = Constant() {r=1}
  n14 = ScalarParameter() {defaultValue=0.1, group=AnglePatch, parameterName=Contour transparency}
  n15 = NamedRerouteUsage()
  n16 = Saturate(None=Multiply(A=MaterialFunctionCall(A=@n12, B=@n13, Transition Phase=@n14, Height Texture=@n15) {materialFunction=&HeightLerp}.Results) {constB=10})
  n17 = LinearInterpolate(B=QualitySwitch(Default=Multiply(A=@n9, B=@n10), Low=Multiply(A=@n8, B=@n11)), Alpha=@n16) {constB=0}
  n18 = OneMinus(None=SmoothStep(Min=@n16, Max=Constant() {r=0.7}, Value=Constant() {r=0.5}))
  Add(A=@n17, B=Saturate(None=Power(Base=@n18) {constExponent=5}))
  n19 = ActorPositionWS()
  ComponentMask(None=@n19) {b=true}
  n20 = Constant()
  n21 = If(
   A=Add(A=PixelDepth().R, B=ScalarParameter() {defaultValue=-1, parameterName=BiasSorting_Far})
   B=ComponentMask(None=SceneTexture() {sceneTextureId=PPI_CustomDepth}.Color) {r=true}
   "A > B"=@n20
   "A == B"=@n20
   "A < B"=Constant() {r=1}
  )
  n22 = Constant() {r=1}
  Multiply(
   A=NamedRerouteUsage()
   B=StaticSwitchParameter(True=ShadingPathSwitch(Default=@n21, Mobile=@n22) {desc=Migrated from FeatureLevelSwitch}, False=@n22) {defaultValue=true, parameterName=CustomDepthSorting}
  )
  n23 = Constant()
  n24 = Constant() {r=0.06}
  n25 = Power(Base=Multiply(A=NamedRerouteUsage(), B=Constant() {r=30}) {constB=2}, Exp=Constant() {r=6})
  NamedRerouteDeclaration(None=Saturate(None=SmoothStep(Min=@n23, Max=@n24, Value=@n25))) {name=TOPMask, nodeColor=<rgba 0 1 0.2969 1>}
  n26 = TextureSample!(UVs, Tex, Level, Apply View MipBias) {
   mipValueMode=TMVM_MipLevel
   samplerType=SAMPLERTYPE_Normal
   texture=/BRCosmetics/Animation/Props/Gadgets_AnglePatch/Textures/T_AnglePatch_Emote_Prop_N
  }
  n27 = Fresnel(
   ExponentIn=Constant() {r=3}
   BaseReflectFractionIn=Constant() {r=0.06}
   Normal=MaterialFunctionCall(BaseNormal=VertexNormalWS(), AdditionalNormal=Transform(None=@n26.RGB)) {materialFunction=&BlendAngleCorrectedNormals}.Result
  )
  NamedRerouteDeclaration(None=Saturate(None=Power(Base=@n27))) {name=Frensnel, nodeColor=<rgba 0.875 0 1 1>}
  n28 = LinearInterpolate(A=Constant() {r=1}, Alpha=NamedRerouteUsage()) {constB=0}
  n29 = NamedRerouteUsage()
  NamedRerouteDeclaration(None=Subtract(A=Subtract(A=@n28, B=@n29), B=NamedRerouteUsage())) {name=SphereMask, nodeColor=<rgba 0.9219 1 0 1>}
  n30 = TextureCoordinate()
  n31 = Constant() {r=60}
  n32 = ScalarParameter() {defaultValue=0.25, group=AnglePatch, parameterName=Grid color intensity}
  n33 = Saturate(None=Divide(A=PixelDepth().R) {constB=5000})
  n34 = Multiply(
   A=TextureSample!(UVs=Multiply(A=@n30, B=@n31) {constB=3}, Tex, Apply View MipBias) {texture=/BRCosmetics/Animation/Props/Gadgets_AnglePatch/Textures/T_AnglePatch_grid}.G
   B=LinearInterpolate(A=LinearInterpolate(A=@n32, Alpha=@n33) {constB=30}, Alpha=NamedRerouteUsage()) {constB=0}
  )
  n35 = NamedRerouteUsage()
  n36 = Constant()
  n37 = NamedRerouteUsage()
  NamedRerouteDeclaration(None=LinearInterpolate(A=Multiply(A=@n34, B=@n35), B=LinearInterpolate(A=@n36, Alpha=@n37) {constA=0.15}, Alpha=NamedRerouteUsage())) {name=GridMask, nodeColor=<rgba 1 0 0.4453 1>}
  n38 = ComponentMask(None=Multiply(
    A=MaterialFunctionCall() {materialFunction=&/MF_PreskinnedWorldPosition_AnglePatch}.Result
    B=Constant() {r=4}
   )) {b=true}
  n39 = TextureSample!(UVs, Tex, Level, Apply View MipBias) {
   mipValueMode=TMVM_MipLevel
   samplerType=SAMPLERTYPE_Normal
   texture=&T_AnglePatch_Emote_Prop_N
  }
  n40 = Multiply(A=@n39.R, B=Constant() {r=0.8}) {constB=2}
  n41 = Divide(A=Abs(None=Add(A=@n38, B=@n40)), B=Constant() {r=5.4})
  n42 = Subtract(A=Multiply(A=Frac(None=@n41)) {constB=2})
  n43 = VertexNormalWS()
  n44 = Constant3Vector() {constant=<rgba 0 0 1 0>}
  n45 = Constant() {r=5}
  n46 = Divide(A=@n45, B=Constant() {r=300})
  n47 = Multiply(A=OneMinus(None=DotProduct(A=@n43, B=@n44)), B=Subtract(A=@n45, B=@n46))
  NamedRerouteDeclaration(None=Divide(A=Abs(None=@n42), B=Add(A=@n47, B=@n46))) {name=line_1, nodeColor=<rgba 1 0 0.3047 1>}
  n48 = DDX(Value=NamedRerouteUsage())
  n49 = DDY(Value=NamedRerouteUsage())
  n50 = Saturate(None=Divide(A=PixelDepth().R) {constB=5000})
  n51 = Multiply(A=Add(A=Abs(None=@n48), B=Abs(None=@n49)), B=LinearInterpolate(A=Constant() {r=1}, B=Constant() {r=5}, Alpha=@n50) {constB=10})
  n52 = SmoothStep(
   Min=Subtract(A=NamedRerouteUsage(), B=@n51)
   Max=Add(A=NamedRerouteUsage(), B=@n51)
   Value=LinearInterpolate(A=Constant() {r=0.01}, B=Constant() {r=1}, Alpha=@n50) {constB=20}
  )
  NamedRerouteDeclaration(None=Saturate(None=Multiply(A=@n52) {constB=2})) {name=Contour line_2, nodeColor=<rgba 0 0.625 1 1>}
  n53 = TextureSample!(UVs, Tex, Apply View MipBias) {samplerSource=SSM_Wrap_WorldGroupSettings, texture=/BRCosmetics/Animation/Props/Gadgets_AnglePatch/Textures/T_AnglePatch_FX}
  NamedRerouteDeclaration(None=@n53.A) {name=EdgeMaskMap, nodeColor=<rgba 0.9219 1 0 1>}
  NamedRerouteDeclaration(None=@n53.G) {name=heightMap, nodeColor=<rgba 0 1 0.2031 1>}
  NamedRerouteDeclaration(None=Constant() {r=1}) {name=depth, nodeColor=<rgba 0 1 0.5078 1>}
  NamedRerouteDeclaration(None=@n53.R) {name=SphereMaskMap, nodeColor=<rgba 0 1 0.0391 1>}
  n54 = ComponentMask(None=Multiply(
    A=MaterialFunctionCall() {materialFunction=&MF_PreskinnedWorldPosition_AnglePatch}.Result
    B=Constant() {r=4}
   )) {b=true}
  n55 = TextureSample!(UVs, Tex, Level, Apply View MipBias) {
   mipValueMode=TMVM_MipLevel
   samplerSource=SSM_Wrap_WorldGroupSettings
   samplerType=SAMPLERTYPE_Normal
   texture=&T_AnglePatch_Emote_Prop_N
  }
  n56 = Multiply(A=@n55.R, B=Constant() {r=0.8}) {constB=2}
  n57 = Divide(A=Abs(None=Add(A=@n54, B=@n56)), B=ScalarParameter() {defaultValue=1, group=AnglePatch, parameterName=Contour spacing})
  n58 = Subtract(A=Multiply(A=Frac(None=@n57)) {constB=2})
  n59 = VertexNormalWS()
  n60 = Constant3Vector() {constant=<rgba 0 0 1 0>}
  n61 = Constant() {r=5}
  n62 = Divide(A=@n61, B=Constant() {r=300})
  n63 = Multiply(A=OneMinus(None=DotProduct(A=@n59, B=@n60)), B=Subtract(A=@n61, B=@n62))
  NamedRerouteDeclaration(None=Divide(A=Abs(None=@n58), B=Add(A=@n63, B=@n62))) {name=line, nodeColor=<rgba 1 0 0.3047 1>}
  NamedRerouteDeclaration(None=@n17) {name=Contour line, nodeColor=<rgba 0 0.625 1 1>}
  NamedRerouteDeclaration(None=@n53.B) {name=Bottom mask, nodeColor=<rgba 0.125 0 1 1>}
  n64 = Saturate(None=NamedRerouteUsage())
  n65 = NamedRerouteUsage()
  n66 = Max(
   A=Saturate(None=TextureSample!(UVs, Tex, Apply View MipBias) {texture=/BRCosmetics/Animation/Props/Gadgets_AnglePatch/Textures/T_AnglePatch_Emote_Prop_D}.R)
   B=Saturate(None=@n65)
  )
  n67 = ScalarParameter() {defaultValue=0.2, group=AnglePatch, parameterName=GreenColorRatio}
  n68 = LinearInterpolate(
   A=VectorParameter() {group=AnglePatch, parameterName=Color_Top, defaultValue=<rgba 0.85 0.3867 0 1>}.RGB
   B=VectorParameter() {group=AnglePatch, parameterName=Color_bottom, defaultValue=<rgba 0.0369 1 0.05 1>}.RGB
   Alpha=Saturate(None=Add(A=@n66, B=@n67) {constB=0.2})
  )
  n69 = Multiply(A=@n68) {constB=0.85}
  n70 = Multiply(A=@n68, B=Constant() {r=1.5}) {constB=20}
  n71 = NamedRerouteUsage()
  n72 = Multiply(A=ScalarParameter() {defaultValue=0.1, group=AnglePatch, parameterName=Contour transparency}) {constB=1.5}
  n73 = Multiply(A=@n65) {constB=20}
  n74 = LinearInterpolate(
   A=VectorParameter() {group=AnglePatch, parameterName=Color_Grid, defaultValue=<rgba 0.1014 1 0.1704 1>}.RGB
   B=Multiply(A=LinearInterpolate(A=@n69, B=@n70, Alpha=@n71), B=Saturate(None=@n72))
   Alpha=Saturate(None=Power(Base=@n73) {constExponent=10})
  )
  NamedRerouteDeclaration(None=MaterialFunctionCall(Emissive=Multiply(A=@n64, B=@n74)) {materialFunction=&MF_FX_EmissiveAdjustments}.Result) {name=Emissive Color, nodeColor=<rgba 0.1953 1 0 1>}
  n75 = Saturate(None=Multiply(A=NamedRerouteUsage(), B=Constant() {r=5}) {constB=3})
  n76 = Add(A=Multiply(A=NamedRerouteUsage(), B=Constant() {r=0.1}) {constB=0.2}, B=NamedRerouteUsage())
  n77 = Constant() {r=0.6}
  n78 = Constant() {r=1}
  n79 = Constant()
  n80 = Saturate(None=Add(A=If(A=@n76, B=@n77, "A > B"=@n78, "A == B"=@n79, "A < B"=@n79) {constB=0.65}, B=@n76))
  NamedRerouteDeclaration(None=LinearInterpolate(A=Multiply(A=@n75, B=@n80), Alpha=NamedRerouteUsage()) {constB=0}) {name=opacity, nodeColor=<rgba 1 0 0.5156 1>}
  NamedRerouteUsage()
  Power(Base=@n65)
  n81 = TextureSample!(UVs, Tex, Level, Apply View MipBias) {
   mipValueMode=TMVM_MipLevel
   samplerType=SAMPLERTYPE_Normal
   texture=&T_AnglePatch_Emote_Prop_N
  }
  n82 = Subtract(A=Add(A=WorldPosition().XYZ, B=Multiply(A=@n81.R) {constB=0}), B=@n19)
  n83 = Constant()
  n84 = Divide(
   A=If(A=@n82, B=ScalarParameter() {defaultValue=1, group=AnglePatch, parameterName=Display threshold}, "A > B"=@n82, "A == B"=@n83, "A < B"=@n83)
   B=Subtract(A=Constant() {r=10}, B=Constant())
  )
  n85 = Constant() {r=30}
  Saturate(None=Power(Base=Multiply(A=@n84, B=@n85) {constB=30}, Exp=Constant() {r=10}))
  VertexInterpolator(VS=PreSkinnedPosition())
 ]
}
